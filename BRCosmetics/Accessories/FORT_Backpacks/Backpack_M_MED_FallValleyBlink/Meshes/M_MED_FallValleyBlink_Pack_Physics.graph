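# Physics asset for the FallValleyBlink medium male backpack cosmetic mesh, giving the pack's dynamic handle bone a simulation body.
# Two sphere bodies cover the pack bone and the dyn_handle_1 bone.
# A single constraint joins the handle to the pack with swing2 and twist locked, so the handle's rotation is confined to the swing1 axis, like a hinge.
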PhysicsAsset /BRCosmetics/Accessories/FORT_Backpacks/Backpack_M_MED_FallValleyBlink/Meshes/M_MED_FallValleyBlink_Pack_Physics
bodies {
 dyn_handle_1 [{shapeType=Sphere, radius=2.4974, shapeName=dyn_handle_1_sphere, center=<0 -0.0009 3.5294>}]
 pack [{shapeType=Sphere, radius=3.5524, center=<0 -3.0708 7.7248>}]
}
constraints [{
  bone1Name=dyn_handle_1
  bone2Name=pack
  swing1LimitDegrees=90
  swing2LimitDegrees=0
  swing2Motion=Locked
  twistLimitDegrees=45
  twistMotion=Locked
 }]
