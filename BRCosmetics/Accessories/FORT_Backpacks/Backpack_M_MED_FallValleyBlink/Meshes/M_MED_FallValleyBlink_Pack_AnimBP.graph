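# Post-process style AnimBlueprint for the FallValleyBlink medium male backpack cosmetic that adds physics motion to the pack's bones.
# The input pose gets an additive offset and scale on root and a roll on dyn_handle_MSTR, then feeds a RigidBody node with overridden gravity and damped component motion.
# Each update eases the FPS Correction Alpha from delta time, lowering the physics blend toward half strength at lower frame rates.
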AnimBlueprint /BRCosmetics/Accessories/FORT_Backpacks/Backpack_M_MED_FallValleyBlink/Meshes/M_MED_FallValleyBlink_Pack_AnimBP
target_skeleton=&/M_MED_FallValleyBlink_Pack_Skeleton
graphs [
 AnimGraph {nodes [
   n1 = "Transform (Modify) Bone - Bone: root"(ComponentPose=Local To Component(LocalPose=Input Pose()), Translation="0,0,-6", Rotation="0,0,0", Scale="0.8,0.9,0.9") {
    rotationMode=BMM_Additive
    scaleMode=BMM_Additive
    translationMode=BMM_Additive
    alphaBoolBlend {blendOption=Linear}
    boneToModify {boneName=root}
    scale=<0.8 0.9 0.9>
    translation {z=-6}
   }
   n2 = "Transform (Modify) Bone - Bone: dyn_handle_MSTR"(ComponentPose=@n1, Alpha=1, Translation="0,0,0", Rotation="0,0,90", Scale="1,1,1") {
    rotationMode=BMM_Additive
    rotationSpace=BCS_BoneSpace
    translationMode=BMM_Additive
    translationSpace=BCS_BoneSpace
    alphaBoolBlend {blendOption=Linear}
    alphaScaleBiasClamp {bInterpResult=true}
    boneToModify {boneName=dyn_handle_MSTR}
    rotation {roll=90}
   }
   n3 = RigidBody(ComponentPose=@n2, Alpha=Get FPS Correction Alpha() -> (FPS Correction Alpha:double)) {
    bClampLinearTranslationLimitToRefPose=true
    bOverrideWorldGravity=true
    overridePhysicsAsset=&/M_MED_FallValleyBlink_Pack_Physics
    simulationSpace=BaseBoneSpace
    alphaBoolBlend {blendOption=Linear}
    alphaScaleBiasClamp {inRange {max=0.015, min=0.03}, outRange {min=0.5}}
    componentAppliedLinearAccClamp=<4000 4000 4000>
    componentLinearAccScale=<0.3 0.3 1>
    componentLinearVelScale=<0.3 0.3 1>
    overrideWorldGravity {z=-980}
   }
   Output Pose(Result=Component To Local(ComponentPose=@n3)) {layerGroup=None}
  ]}
 EventGraph {nodes [
   n1 = Event BlueprintUpdateAnimation()
   > Set FPS Correction Alpha(FPS Correction Alpha:double=FInterpTo(
     Current=Get FPS Correction Alpha()
     Target=MapRangeClamped(Value=@n1.DeltaTimeX, InRangeA=0.03, InRangeB=0.015, OutRangeA=0.5)
     DeltaTime=@n1.DeltaTimeX
     InterpSpeed=0.1
    )) -> (Output_Get:double)
   TryGetPawnOwner()
  ]}
]
variables [FPS Correction Alpha {type=number}]
